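; B_AutoAimWeaponPawnComponent_Base is a Blueprint base pawn component for the auto-aim weapon feature.
; Its graph exposes a custom DestroyComponent event that removes the component only on the network authority.
; The event wraps K2_DestroyComponent because, per its comment, that function fails when invoked from an external class.
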(event Custom|DestroyComponent
  ;[0.09488 0.1198 0.08941] DestroyComponent: Blueprint event to help invoke K2_DestroyComponent() function which doesNOT work if invoked from an external class.
  (Networking|SwitchHasAuthority
    (:Authority
      (Components|DestroyComponent))))
