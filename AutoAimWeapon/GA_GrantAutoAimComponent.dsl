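; Gameplay ability that grants an auto-aim weapon component to the avatar's Fortnite player pawn for as long as the ability is active.
; On activation the server adds the configured component class if the pawn lacks one, then stores the new instance cast to the auto-aim base component.
; On end the server destroys the stored component so it cannot conflict with another auto-aim weapon, and the parent end logic runs.
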(event Ability|EventActivateAbility
  (bind _returnvalue (CallFunction|GetAvatarasFortPlayerPawn))
  ;[0.01374 0.1458 0.01462] ActivateAbility: On the server, add the auto aim component class to the player pawn, if it doesn't exist.
  (if (Networking|IsServer)
    (Utilities|IsValid _returnvalue
      (:Is%20Valid
        (Utilities|IsValid (Actor|GetComponentbyClass _returnvalue (Variables|AutoAim|GetComponent%20Class))
          (:Is%20Not%20Valid
            (bind _returnvalue_2 (Game|AddComponentbyClass (CallFunction|GetAvatarasFortPlayerPawn) (Variables|AutoAim|GetComponent%20Class)))
            (bind _asb_auto_aim_weapon_pawn_component_base (Utilities|Casting|CastToB_AutoAimWeaponPawnComponent_Base _returnvalue_2))
            (Variables|Default|SetAutoAimPistolComp _asb_auto_aim_weapon_pawn_component_base)))))))

(event Ability|EventOnEndAbility
  (params (bWasCancelled :type Boolean))
  ;[0.07292 0.03969 0.001899] EndAbility: Remove the component from the player pawn so that it doesn't conflict with other auto-aim weapon that may be equipped next.
  (Utilities|FlowControl|Sequence
    (:then_0
      (if (Networking|IsServer)
        (bind _autoaimpistolcomp (Variables|Default|GetAutoAimPistolComp)
          (:then
            (Class|BAutoAimWeaponPawnComponentBase|DestroyComponent _autoaimpistolcomp))
          (:else))))
    (:then_1
      (|CallParentFunction bWasCancelled))))
